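# Translucent, lit Niagara sprite material that draws animated fly particles.
# A time-driven two-frame SubUV flipbook samples color, alpha and normal textures at an adjustable speed.
# The Front Facing dynamic parameter blends the sprite toward a small sphere-mask dot and flat fly color, suggesting a head-on view.
# The CustomColor dynamic parameter blends in particle-color tinting, masked by the texture's boosted luminance on the side view.
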Material /Game/Effects/Fort_Effects/Effects/Bugs/Flies/M_ParticleFlies
graph {
 properties {
  bUsedWithNiagaraSprites=true
  blendMode=BLEND_Translucent
  pixelDepthOffsetMode=PDOM_Legacy
  previewMesh=/Game/Environments/Hera/FX/Meshes/SM_WaterTankBreak_GlassCracks_Water_01
  translucencyLightingMode=TLM_VolumetricDirectional
  usedShadingModels=Default Lit
 }
 nodes [
  n1 = DynamicParameter() {defaultValue=<rgba 0 0 0 0>, paramNames [Front Facing, CustomColor, "", ""]}
  n2 = Constant2Vector() {g=1, r=2}
  n3 = TextureCoordinate()
  n4 = Multiply(A=Time(), B=ScalarParameter() {defaultValue=20, parameterName=Speed})
  n5 = MaterialFunctionCall!(Texture=TextureObject() {texture=&/T_ParticleFlies_M}, UVs=@n3, SubImages=@n2, Frame=@n4) {materialFunction=/Engine/Functions/Engine_MaterialFunctions03/Texturing/SubUV_Function}
  LinearInterpolate(
   A=@n5.Alpha
   B=SphereMask(A=TextureCoordinate(), B=Constant() {r=0.5}, Radius=ScalarParameter() {defaultValue=0.15, parameterName=Fly Front Radius})
   Alpha=@n1.Front Facing
  )
  n6 = ParticleColor()
  n7 = Desaturation(None=@n5.RGB)
  n8 = Saturate(None=Multiply(A=LinearInterpolate(A=@n7) {constAlpha=0.01}) {constB=10})
  n9 = Add(
   A=ScalarParameter() {defaultValue=0.025, parameterName=Base Color Add}
   B=LinearInterpolate(
    A=@n5.RGB
    B=Multiply(A=@n6.RGB, B=@n8)
    Alpha=DynamicParameter() {defaultValue=<rgba 0 0 0 0>, paramNames [Front Facing, CustomColor, "", ""]}.CustomColor
   )
  )
  n10 = LinearInterpolate(
   A=VectorParameter() {parameterName=Front Facing Fly Color, defaultValue=<rgba 0.0181 0.0365 0.0212 1>}.RGB
   B=Multiply(A=ParticleColor().RGB) {constB=0.1}
   Alpha=DynamicParameter() {defaultValue=<rgba 0 0 0 0>, paramNames [Front Facing, CustomColor, "", ""]}.CustomColor
  )
  LinearInterpolate(
   B=Multiply(A=LinearInterpolate(A=@n9, B=@n10, Alpha=@n1.Front Facing)) {constB=0.5}
   Alpha=DynamicParameter() {defaultValue=<rgba 0 0 0 0>, paramNames [Front Facing, CustomColor, "", ""]}.CustomColor
  )
  MaterialFunctionCall(
   Texture=TextureObject() {samplerType=SAMPLERTYPE_Normal, texture=&/T_ParticleFlies_N}
   UVs=@n3
   SubImages=@n2
   Frame=@n4
  ) {materialFunction=&SubUV_Function}
 ]
}
