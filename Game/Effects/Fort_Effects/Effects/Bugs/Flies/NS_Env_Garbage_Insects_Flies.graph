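# Niagara environment effect of small sprite flies buzzing around garbage, with user-set spawn rate and optional custom color.
# Flies spawn probabilistically in a small owner-scaled sphere, live several seconds and scale in and out.
# Two panning curl noise forces, upward gravity and drag drive erratic flight, while an HLSL-driven point force acts as a leash, pulling harder the farther flies stray from the owner.
# UseCustomColor feeds color, material parameter and a visibility tag.
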NiagaraSystem /Game/Effects/Fort_Effects/Effects/Bugs/Flies/NS_Env_Garbage_Insects_Flies
dependencies {usedDataInterfaces [&NiagaraDataInterfaceCurve], usedDynamicInputs [], usedModules [&SystemState]}
emitters {Flies {
  emitterUpdateScript [
   &EmitterState()
   &SpawnRate(Spawn Probability=0.3, SpawnRate=&Multiply_Float(A=3, B=$User.SpawnRate), Use Spawn Probability=true)
  ]
  particleSpawnScript [
   &InitializeParticle(
    Color=&LerpLinearColors(StartColor=<rgba 1 1 1 1>, EndColor=$User.Color, LerpFactor=$User.UseCustomColor)
    Lifetime Max=10
    Lifetime Min=6
    Lifetime Mode=::NewEnumerator1
    Mass Mode=::NewEnumerator3
    Sprite Rotation Mode=::NewEnumerator2
    Sprite Size Mode=::NewEnumerator1
    Uniform Sprite Size Max=4
    Uniform Sprite Size Min=3
    UsePositionOffset=false
    Write Material Random=true
   )
   &ShapeLocation(MaskSpawnGroup=false, Override Seed=false, Sphere Radius=35, UseOwnerScale=true)
   /SetVariables(
    Particles.FlyLeashAmount=35
    Particles.FlyVisibilityTag:Vector4f=&MakeVector4(X=1, Y=$User.UseCustomColor, Z=0, W=0)
   ) {isSetParameters=true}
   &DynamicMaterialParameters(
    Index 0 Param 2=$User.UseCustomColor
    Module.Index0Param0Write=true
    Module.Index0Param1Write=true
    Module.Index0Param2Write=true
    Module.Index0Param3Write=true
   )
  ]
  particleUpdateScript [
   &ParticleState()
   &CurlNoiseForce(Noise Frequency=75, Noise Strength=500, Pan Noise=true, Pan Noise Field=<0.2 0.3 0.4>)
   &CurlNoiseForce(Noise Frequency=75, Noise Strength=100, Pan Noise=true, Pan Noise Field=<0.1 0.2 0.3>) {module=CurlNoiseForce001}
   &GravityForce(Gravity=<0 0 100>)
   &PointForce(
    Force Strength=HlslExpression() {hlslExpression=" max(0,(length(Engine.Owner.Position - Particles.Position) - Particles.FlyLeashAmount )) * -10"}
    Module.Use Falloff=false
    Module.UseLinearForce=false
   )
   &Drag(Use Linear Drag=true, Use Rotational Drag=false)
   &ScaleSpriteSize(Uniform Curve Sprite Scale=DataInterface() {propertyValues {Curve {keys [
        <at 0 0 RCIM_Cubic>
        {arriveTangent=0.0309, interpMode=RCIM_Cubic, leaveTangent=0.0309, tangentMode=RCTM_User, time=0.05, value=1}
        {arriveTangent=-0.0071, interpMode=RCIM_Cubic, leaveTangent=-0.0071, tangentMode=RCTM_User, time=0.95, value=1}
        <at 1 0 RCIM_Cubic>
       ]}}})
   &SolveForcesAndVelocity(Clamp Velocity=false, Limit Acceleration=false)
  ]
  rendererClasses [&NiagaraSpriteRendererProperties]
  renderers [
   {rendererClass=&NiagaraSpriteRendererProperties}
   {rendererClass=&NiagaraSpriteRendererProperties, rendererIndex=1}
  ]
 }}
summary {
 systemName=NS_Env_Garbage_Insects_Flies
 emitters [{emitterName=Flies, rendererClasses [&NiagaraSpriteRendererProperties]}]
 userVariables [
  User.Color {type=LinearColor, description="", defaultValue {struct=LinearColor, value=<rgba 0.1214 1 0 1>}}
  User.SpawnRate {type=NiagaraFloat, defaultValue=1, description=""}
  User.UseCustomColor {type=NiagaraFloat, defaultValue=0, description=""}
 ]
}
